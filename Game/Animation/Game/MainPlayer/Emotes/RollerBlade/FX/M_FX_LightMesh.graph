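# M_FX_LightMesh is an unlit, two-sided translucent Niagara mesh-particle material for a light effect in the player's RollerBlade emote.
# Particle color feeds an emissive adjustment function.
# A panning Perlin texture, added to a powered inverted-V gradient, is multiplied by a second gradient with adjustable power and an inverted Fresnel that fades silhouette edges.
# Particle alpha scales the resulting mask.
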Material /Game/Animation/Game/MainPlayer/Emotes/RollerBlade/FX/M_FX_LightMesh
graph {
 properties {
  bUsedWithNiagaraMeshParticles=true
  blendMode=BLEND_Translucent
  pixelDepthOffsetMode=PDOM_Legacy
  shadingModel=MSM_Unlit
  twoSided=true
  usedShadingModels=Unlit
 }
 nodes [
  n1 = ParticleColor()
  MaterialFunctionCall(Emissive=@n1.RGB) {materialFunction=&MF_FX_EmissiveAdjustments}
  n2 = TextureCoordinate()
  n3 = VectorParameter() {parameterName=T1 Tile Speed, defaultValue=<rgba 3 0 0.2 0.05>}
  n4 = MaterialFunctionCall(X=@n3.R, Y=@n3.R) {materialFunction=&MakeFloat2}
  n5 = TextureSample(UVs=Panner(
    Coordinate=Multiply(A=@n2, B=@n4.Result)
    Speed=MaterialFunctionCall(X=@n3.B, Y=@n3.B) {materialFunction=&MakeFloat2}.Result
   )) {samplerType=SAMPLERTYPE_Masks, texture=/Game/Characters/Player/Female/Medium/Bodies/F_MED_Neon_Tight_Suit/Materials/Soft_Perlin_output}
  n6 = OneMinus(None=MaterialFunctionCall(Float2=TextureCoordinate()) {materialFunction=&BreakOutFloat2Components}.G)
  n7 = Power(Base=@n6) {constExponent=5}
  n8 = Saturate(None=Add(A=@n5.R, B=Saturate(None=@n7)))
  n9 = ScalarParameter() {defaultValue=4, parameterName=Power}
  n10 = Fresnel(ExponentIn=ScalarParameter() {defaultValue=1, parameterName=Exponent}) {baseReflectFraction=0}
  n11 = Multiply(A=Saturate(None=Power(Base=@n6, Exp=@n9)), B=Saturate(None=OneMinus(None=@n10)))
  Multiply(A=Saturate(None=Multiply(A=@n8, B=@n11)), B=@n1.A)
 ]
}
